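# Niagara idle effect for the GainSneeze pickaxe that sheds small, short-lived sprite drops.
# One Drops emitter spawns probabilistically and splits particles into three partitions, each with its own brown or off-white color and local offset, with one offset mirrored for the front end.
# Drops inherit some velocity, then fall under gravity and random drag while scaling in size.
# A system-level SystemSpeedAlpha multiplies their opacity, tying visibility to a speed value.
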NiagaraSystem /BRCosmetics/Weapons/FORT_Melee/Pickaxe_GainSneeze/FX/NS_Pickaxe_GainSneeze_Idle
dependencies {
 usedDataInterfaces [
  &NiagaraDataInterfaceCurve
  &NiagaraDataInterfaceVector2DCurve
  &NiagaraDataInterfaceSocketReader
 ]
 usedDynamicInputs [
  /Game/Effects/Niagara/DynamicInputs/TimeOfDay/DaySequenceTimeOfDayColor
  &Divide_Float
  &VectorLength
 ]
 usedModules [
  &SystemState
  &self:SystemSpawnScript.SystemScriptSource.SystemScriptGraph.NiagaraNodeAssignment_0.SetVariables_#1
  &self:SystemSpawnScript.SystemScriptSource.SystemScriptGraph.NiagaraNodeAssignment_4.SetVariables_#2
  &self:SystemSpawnScript.SystemScriptSource.SystemScriptGraph.NiagaraNodeAssignment_3.SetVariables_#3
 ]
}
emitters {Drops {
  emitterUpdateScript [
   &EmitterState()
   &SpawnRate(Spawn Probability=0.2, Use Spawn Probability=true)
  ]
  particleSpawnScript [
   &PartitionParticles(Number Of Partitions=3, Partition Mode:ENiagara_PartitionMode_v2=::NewEnumerator5)
   &InitializeParticle(
    Color=&MergeHelper_ColorSwitch(
     0=<rgba 0.1927 0.0837 0.0452 1>
     1=<rgba 0.9216 0.9216 0.9216 1>
     2=<rgba 0.1927 0.0837 0.0452 1>
     3=?
     Index=$Particles.PartitionParticles.Partition
     Num=::NewEnumerator1
    )
    Lifetime Max=0.6
    Lifetime Min=0.3
    Lifetime Mode=::NewEnumerator1
    Position Offset=&MergeHelper_VectorSwitch(
     0=&FN_FrontEndVector3Modifier(Gameplay=<-1 0 25>, Front End=<1 0 25>)
     1=<-1 0 18>
     2=<-1 0 10>
     3=?
     Index=$Particles.PartitionParticles.Partition
     Num=::NewEnumerator1
    )
    Position Offset Coordinate Space=::Local
    Sprite Size Mode=::NewEnumerator1
    Uniform Sprite Size Max=2.4
    Uniform Sprite Size Min=2
    UsePositionOffset=true
    Write Material Random=true
   )
   /SetVariables(Particles.Color=&MultiplyLinearColor(Linear Color=$Particles.Color, Scale Factor=&MakeVector4(X=1, Y=1, Z=1, W=$System.SystemSpeedAlpha))) {isSetParameters=true}
   &InheritVelocity(Inherited Velocity Amount Scale=<0.25 0.25 0.25>, Inherited Velocity Speed Limit=20, Speed Limit=false, UseSpeedThreshold=false)
  ]
  particleUpdateScript [
   &ParticleState()
   &Drag(Drag=&RandomRangeFloat(Minimum=1, Maximum=7), Use Linear Drag=true, Use Rotational Drag=false)
   &GravityForce(Gravity=<0 0 -500>)
   &ScaleColor(
    Scale Alpha=&Multiply_Float(
     A=&FloatFromCurve(FloatCurve=DataInterface() {propertyValues {Curve {
         defaultValue=0
         keys [{interpMode=RCIM_Cubic, tangentMode=RCTM_User, time=0.4403, value=1}, {interpMode=RCIM_Cubic, tangentMode=RCTM_User, time=1, value=0}]
        }}})
     B=$System.SystemSpeedAlpha
    )
    ScaleA=true
    ScaleRGB=true
   )
   &ScaleSpriteSize(Uniform Curve Sprite Scale=DataInterface() {propertyValues {Curve {keys [
        {arriveTangent=2.753, interpMode=RCIM_Cubic, leaveTangent=2.753, tangentMode=RCTM_User, time=0, value=0.3494}
        {arriveTangent=0.143, interpMode=RCIM_Cubic, leaveTangent=0.143, tangentMode=RCTM_User, time=0.3012, value=0.9559}
        {interpMode=RCIM_Cubic, tangentMode=RCTM_User, time=1, value=0.1882}
       ]}}}) {module=ScaleSpriteSize001}
   &SolveForcesAndVelocity(Clamp Velocity=false, Limit Acceleration=false)
  ]
  rendererClasses [&NiagaraSpriteRendererProperties]
  renderers [{rendererClass=&NiagaraSpriteRendererProperties}]
 }}
summary {
 systemName=NS_Pickaxe_GainSneeze_Idle
 emitters [{emitterName=Drops, rendererClasses [&NiagaraSpriteRendererProperties]}]
 userVariables [
  User.bIsFrontEnd {type=NiagaraBool}
  User.bIsFrontEndPreview {type=NiagaraBool}
  User.PhysicalSurface {type=NiagaraInt32}
  User.RibbonIDOffset {type=NiagaraInt32}
 ]
}
