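# Animation Blueprint for the GainSneeze pickaxe that adds springy, gravity-free secondary motion to its three dyn_scoop bones.
# The cached input pose feeds two Anim Dynamics chains, one in default simulation space and one RootRelative, with tight angular limits.
# A Blend Poses by bool picks between the chains, with no condition shown wired.
# The update event lowers AnimDynAlpha in the front end, but the dynamics nodes show a fixed Alpha.
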AnimBlueprint /BRCosmetics/Weapons/FORT_Melee/Pickaxe_GainSneeze/Meshes/GainSneeze_Axe_AnimBP
target_skeleton=&/GainSneeze_Axe_Skeleton
graphs [
 AnimGraph {nodes [
   n1 = Anim Dynamics - Bone: dyn_scoop_00(ComponentPose=Local To Component(LocalPose=Use cached pose 'Cache_Input'()), Alpha=1, ExternalForce="0,0,0") {
    angularDampingOverride=1
    angularSpringConstant=20000
    bAngularSpring=true
    bOverrideAngularDamping=true
    bUsePlanarLimit=false
    gravityScale=0
    linearSpringConstant=30000
    boundBone {boneName=dyn_scoop_00}
    componentLinearAccScale=<1 1 1>
    componentLinearVelScale=<1 1 1>
    physicsBodyDefinitions [{
      boundBone {boneName=dyn_scoop_00}
      boxExtents=<40 40 40>
      constraintSetup {angularLimitsMax=<0 8.5 8.5>, angularLimitsMin=<0 -8.5 -8.5>, angularTarget=<1 0 0>}
      localJointOffset=<3 0 0>
     }]
    relativeSpaceBone {boneName=b_AR_Root}
    simSpaceSettings {maxAngularAcceleration=50, maxAngularVelocity=500, simSpaceAngularAlpha=1}
   }
   n2 = Anim Dynamics - Bone: dyn_scoop_01(ComponentPose=@n1, Alpha=1, ExternalForce="0,0,0") {
    angularDampingOverride=0.99
    angularSpringConstant=35000
    bAngularSpring=true
    bOverrideAngularDamping=true
    bUsePlanarLimit=false
    gravityScale=0
    boundBone {boneName=dyn_scoop_01}
    componentLinearAccScale=<1 1 1>
    componentLinearVelScale=<1 1 1>
    physicsBodyDefinitions [{
      boundBone {boneName=dyn_scoop_01}
      boxExtents=<35 35 35>
      constraintSetup {angularLimitsMax=<0 2 1>, angularLimitsMin=<0 -1.5 -2>, angularTarget=<1 0 0>}
      localJointOffset=<8 0 0>
     }]
    relativeSpaceBone {boneName=b_AR_Root}
    simSpaceSettings {maxAngularAcceleration=5000, maxAngularVelocity=1000, simSpaceAngularAlpha=1}
   }
   n3 = Anim Dynamics - Bone: dyn_scoop_02(ComponentPose=@n2, Alpha=1, ExternalForce="0,0,0") {
    angularDampingOverride=0.9
    angularSpringConstant=10000
    bAngularSpring=true
    bOverrideAngularDamping=true
    bUsePlanarLimit=false
    gravityScale=0
    boundBone {boneName=dyn_scoop_02}
    componentLinearAccScale=<1 1 1>
    componentLinearVelScale=<1 1 1>
    physicsBodyDefinitions [{
      boundBone {boneName=dyn_scoop_02}
      boxExtents=<30 30 30>
      constraintSetup {angularLimitsMax=<0 1 1>, angularLimitsMin=<0 -1 -1>, angularTarget=<1 0 0>}
      localJointOffset=<7 0 0>
     }]
    relativeSpaceBone {boneName=b_AR_Root}
    simSpaceSettings {maxAngularAcceleration=5000, maxAngularVelocity=1000, simSpaceAngularAlpha=1}
   }
   n4 = Anim Dynamics - Bone: dyn_scoop_00(ComponentPose=Local To Component(LocalPose=Use cached pose 'Cache_Input'()), Alpha=1, ExternalForce="0,0,0") {
    angularDampingOverride=1
    angularSpringConstant=20000
    bAngularSpring=true
    bOverrideAngularDamping=true
    bUsePlanarLimit=false
    gravityScale=0
    linearSpringConstant=30000
    simulationSpace=RootRelative
    boundBone {boneName=dyn_scoop_00}
    componentLinearAccScale=<1 1 1>
    componentLinearVelScale=<1 1 1>
    physicsBodyDefinitions [{
      boundBone {boneName=dyn_scoop_00}
      boxExtents=<40 40 40>
      constraintSetup {angularLimitsMax=<0 8.5 8.5>, angularLimitsMin=<0 -8.5 -8.5>, angularTarget=<1 0 0>}
      localJointOffset=<3 0 0>
     }]
    relativeSpaceBone {boneName=b_AR_Root}
    simSpaceSettings {maxAngularAcceleration=50, maxAngularVelocity=500}
   }
   n5 = Anim Dynamics - Bone: dyn_scoop_01(ComponentPose=@n4, Alpha=1, ExternalForce="0,0,0") {
    angularDampingOverride=0.99
    angularSpringConstant=35000
    bAngularSpring=true
    bOverrideAngularDamping=true
    bUsePlanarLimit=false
    gravityScale=0
    simulationSpace=RootRelative
    boundBone {boneName=dyn_scoop_01}
    componentLinearAccScale=<1 1 1>
    componentLinearVelScale=<1 1 1>
    physicsBodyDefinitions [{
      boundBone {boneName=dyn_scoop_01}
      boxExtents=<35 35 35>
      constraintSetup {angularLimitsMax=<0 2 1>, angularLimitsMin=<0 -1.5 -2>, angularTarget=<1 0 0>}
      localJointOffset=<8 0 0>
     }]
    relativeSpaceBone {boneName=b_AR_Root}
    simSpaceSettings {maxAngularAcceleration=5000, maxAngularVelocity=1000}
   }
   n6 = Anim Dynamics - Bone: dyn_scoop_02(ComponentPose=@n5, Alpha=1, ExternalForce="0,0,0") {
    angularDampingOverride=0.9
    angularSpringConstant=10000
    bAngularSpring=true
    bOverrideAngularDamping=true
    bUsePlanarLimit=false
    gravityScale=0
    simulationSpace=RootRelative
    boundBone {boneName=dyn_scoop_02}
    componentLinearAccScale=<1 1 1>
    componentLinearVelScale=<1 1 1>
    physicsBodyDefinitions [{
      boundBone {boneName=dyn_scoop_02}
      boxExtents=<30 30 30>
      constraintSetup {angularLimitsMax=<0 1 1>, angularLimitsMin=<0 -1 -1>, angularTarget=<1 0 0>}
      localJointOffset=<7 0 0>
     }]
    relativeSpaceBone {boneName=b_AR_Root}
    simSpaceSettings {maxAngularAcceleration=5000, maxAngularVelocity=1000}
   }
   Output Pose(Result=Blend Poses by bool(BlendPose_0=Component To Local(ComponentPose=@n6), BlendPose_1=Component To Local(ComponentPose=@n3))) {layerGroup=None}
   AnimGraphNode_SubInput_0 = Input Pose()
   Save cached pose 'Cache_Input'(Pose=@AnimGraphNode_SubInput_0) {cacheName=Cache_Input}
  ]}
 EventGraph {nodes [
   Event BlueprintUpdateAnimation()
   > n1 = Branch(Condition=Get bIsInFrontEnd())
   > Set AnimDynAlpha(AnimDynAlpha:double=0.4) -> (Output_Get:double)
   Set AnimDynAlpha(execute=@n1.else, AnimDynAlpha=1)
   TryGetPawnOwner()
  ]}
]
variables [AnimDynAlpha {type=number}]
